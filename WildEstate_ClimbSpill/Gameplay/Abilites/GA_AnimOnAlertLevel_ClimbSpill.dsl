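; Gameplay ability that mirrors an AI bot's alert level as loose animation gameplay tags on its pawn.
; On activation it binds to the controller's alert-level change event and evaluates the current level immediately, unbinding when the ability ends.
; Unaware clears the Alerted and Combat tags, Threatened swaps suspicious tags for threatened ones, and a downed (DBNO) pawn gets the reverse swap.
; ApplyTags adds only missing tags and removes only present ones.
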(event Ability|EventActivateAbility
  (|CallParentFunction)
  (Ability|CommitAbility)
  (AlertLevel|BindEventtoOnAlertLevelChangedEvent (Variables|InternalData|GetAIController) (delegate Custom|OnAlertLevelChanged))
  (CallFunction|OnAlertLevelChanged (Variables|InternalData|GetAIController) "Unaware" (Variables|/Script/FortniteAI.FortAthenaAIBotController|GetCurrentAlertLevel (Variables|InternalData|GetAIController))))

(event Custom|OnAlertLevelChanged
  (params (BotController :type Fort%20Athena%20AIBot%20Controller%20Object%20Reference) (OldAlertLevel :type EAlertLevel%20Enum :default "Unaware") (NewAlertLevel :type EAlertLevel%20Enum :default "Unaware"))
  (bind _suspiciousgameplaytag (Variables|Default|GetSuspiciousGameplayTag))
  (bind _threatenedgameplaytag (Variables|Default|GetThreatenedGameplayTag))
  (if (not (Variables|/Script/FortniteGame.FortPawn|GetbIsDBNO (Variables|InternalData|GetAIPawn)))
    (switch Utilities|FlowControl|Switch|SwitchonEAlertLevel NewAlertLevel
      (:Unaware
        (CallFunction|ApplyTags :RemoveTags "(GameplayTags=((TagName=\"Event.Animation.Alerted\"),(TagName=\"Event.Animation.Combat\")))"))
      (:Suspicious
        _)
      (:LostTarget
        _)
      (:Threatened
        (CallFunction|ApplyTags _threatenedgameplaytag _suspiciousgameplaytag)))
    (CallFunction|ApplyTags _suspiciousgameplaytag _threatenedgameplaytag)))

(event Custom|ApplyTags
  (params (AddTags :type Gameplay%20Tag%20Container%20Structure) (RemoveTags :type Gameplay%20Tag%20Container%20Structure))
  (Utilities|FlowControl|Sequence
    (:then_0
      (for _array_element (GameplayTags|BreakGameplayTagContainer AddTags)
        (if (not (GameplayTags|HasMatchingGameplayTag (Variables|InternalData|GetAIPawn) _array_element))
          (Ability|GameplayEffect|AddLooseGameplayTags (Variables|InternalData|GetAIPawn) (GameplayTags|MakeGameplayTagContainerfromTag _array_element) true))))
    (:then_1
      (for _array_element_2 (GameplayTags|BreakGameplayTagContainer RemoveTags)
        (if (GameplayTags|HasMatchingGameplayTag (Variables|InternalData|GetAIPawn) _array_element_2)
          (Ability|GameplayEffect|RemoveLooseGameplayTags (Variables|InternalData|GetAIPawn) (GameplayTags|MakeGameplayTagContainerfromTag _array_element_2) true))))))

(event Ability|EventOnEndAbility
  (params (bWasCancelled :type Boolean))
  (AlertLevel|UnbindEventfromOnAlertLevelChangedEvent (Variables|InternalData|GetAIController) (delegate Custom|OnAlertLevelChanged))
  (|CallParentFunction bWasCancelled))
